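# MetaSound patch for flight banking audio, playing a movement layer and random one-shots as DirectionValue crosses their threshold gates.
# The loop gate drives a fade envelope that starts and stops a random movement wave, pitched by normalized speed and scaled by SpeedVol.
# That envelope shapes the stereo mix of both layers, so it also attenuates one-shots.
# The Mobile page keeps only mono one-shots on Out L, scaled by speed.
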MetaSoundPatch /FortMetasound/Abilities/Flight/MSP_Ability_Flight_Bank
inputs [
 DirectionValue {type=Float}
 EnterCondition {type=Enum:TriggerComparisonType}
 ExitCondition {type=Enum:TriggerComparisonType}
 Fade {type=Float}
 LoopThreshold {type=Float}
 LoopVolume {type=Float}
 MaxPitchMod {type=Float}
 MinPitchMod {type=Float}
 MovementSounds {type=WaveAsset:Array}
 NormSpeed {type=Float}
 OneShotRetriggerDelay {type=Time}
 OneShotSounds {type=WaveAsset:Array}
 OneshotThreshold {type=Float}
 OneShotVolume {type=Float}
 SpeedVol {type=Float}
]
outputs [
 OnLoopPlay {type=Trigger}
 OnLoopStop {type=Trigger}
 OnOneShotPlayed {type=Trigger}
 Out L {type=Audio}
 Out R {type=Audio}
]
pages [
 Default {
  is_build_graph=true
  input_defaults {
   DirectionValue=1
   EnterCondition=5
   ExitCondition=2
   Fade=0.25
   LoopThreshold=0.3
   LoopVolume=1
   MaxPitchMod=1.1
   MinPitchMod=0.9
   NormSpeed=0
   OneShotRetriggerDelay=0.2
   OneShotVolume=1
   OneshotThreshold=0.1
   SpeedVol=1
   MovementSounds []
   OneShotSounds []
  }
  nodes [
   n1 = UE.ConversionFloatToTime(In=^Fade)
   n2 = &MSP_OneshotGate(
    Value=^DirectionValue
    EnterCondition=^EnterCondition
    ExitCondition=^ExitCondition
    OneshotThreshold=^LoopThreshold
    RetriggerDelay=^OneShotRetriggerDelay
   )
   ADSREnvelopeADSREnvelope_0 = ADSR Envelope.ADSR Envelope:Float(Attack Time=@n1.Out, Release Time=@n1.Out, Trigger Attack=@n2.OnGateEnter, Trigger Release=@n2.OnGateExit) {also {defaults {Sustain Level=1}}}
   n3 = InitVariable.Trigger()
   ArrayRandomGet_0 = Array.Random Get:WaveAsset:Array(
    In Array=^MovementSounds
    Next=VariableAccessor.Trigger(Variable=VariableMutator.Trigger(Value=@ADSREnvelopeADSREnvelope_0.On Attack Triggered, Variable=@n3.Variable).Variable).Value
   )
   FrequencyMultipliertoSemitone_0 = UE.Frequency Multiplier to Semitone:Float(Frequency Multiplier=MapRange.MapRange:Float(Out Range A=^MaxPitchMod, Out Range B=^MinPitchMod, In=^NormSpeed).Out Value)
   GetWaveDuration_0 = "UE.Get Wave Duration: "(Wave=@ArrayRandomGet_0.Value)
   n4 = InitVariable.Trigger()
   WavePlayer_0 = UE.Wave Player:Stereo(
    Wave Asset=@ArrayRandomGet_0.Value
    Pitch Shift=@FrequencyMultipliertoSemitone_0.Semitones
    Play=UE.RandomTime(
     Next=@ArrayRandomGet_0.On Next
     Max=UE.Multiply:Time by Float(PrimaryOperand=@GetWaveDuration_0.Duration) {also {defaults {AdditionalOperands=0.5}}}.Out
    ).On Next
    Stop=VariableAccessor.Trigger(Variable=VariableMutator.Trigger(Value=@ADSREnvelopeADSREnvelope_0.On Done, Variable=@n4.Variable).Variable).Value
   ) {also {defaults {Loop="<MetaSoundBuilderResult.SUCCEEDED: 0>"}}}
   MSP_OneshotGate_0 = &MSP_OneshotGate(
    Value=^DirectionValue
    EnterCondition=^EnterCondition
    ExitCondition=^ExitCondition
    RetriggerDelay=^OneShotRetriggerDelay
    OneshotThreshold=^OneshotThreshold
   )
   ArrayRandomGet_1 = Array.Random Get:WaveAsset:Array(In Array=^OneShotSounds, Next=@MSP_OneshotGate_0.OnGateEnter)
   WavePlayer_1 = UE.Wave Player:Stereo(Play=@ArrayRandomGet_1.On Next, Wave Asset=@ArrayRandomGet_1.Value)
   AudioMixerAudioMixerStereo2_0 = "AudioMixer.Audio Mixer (Stereo, 2)"(
    Gain 0=^LoopVolume
    Gain 1=^OneShotVolume
    In 0 L=UE.Multiply:Audio by Float(AdditionalOperands=^SpeedVol, PrimaryOperand=@WavePlayer_0.Out Left).Out
    In 0 R=UE.Multiply:Audio by Float(AdditionalOperands=^SpeedVol, PrimaryOperand=@WavePlayer_0.Out Right).Out
    In 1 L=@WavePlayer_1.Out Left
    In 1 R=@WavePlayer_1.Out Right
   )
   n5 = VariableAccessor.Float(Variable=VariableMutator.Float(Value=@ADSREnvelopeADSREnvelope_0.Out Envelope, Variable=InitVariable.Float().Variable).Variable)
   Multiply_0 = UE.Multiply:Audio by Float(PrimaryOperand=@AudioMixerAudioMixerStereo2_0.Out R, AdditionalOperands=@n5.Value)
   Multiply_2 = UE.Multiply:Audio by Float(PrimaryOperand=@AudioMixerAudioMixerStereo2_0.Out L, AdditionalOperands=@n5.Value)
  ]
  links [
   {dst="@OnOneShotPlayed", src=MSP_OneshotGate_0.OnGateEnter}
   {dst="@Out R", src=Multiply_0.Out}
   {dst="@Out L", src=Multiply_2.Out}
   {dst="@OnLoopStop", src=WavePlayer_0.On Finished}
   {dst="@OnLoopPlay", src=WavePlayer_0.On Play}
  ]
 }
 Mobile {
  nodes [
   InitVariable.Float()
   InitVariable.Trigger()
   InitVariable.Trigger()
   MSP_OneshotGate_0 = &MSP_OneshotGate(
    Value=^DirectionValue
    EnterCondition=^EnterCondition
    ExitCondition=^ExitCondition
    RetriggerDelay=^OneShotRetriggerDelay
    OneshotThreshold=^OneshotThreshold
   )
   ArrayRandomGet_0 = Array.Random Get:WaveAsset:Array(In Array=^OneShotSounds, Next=@MSP_OneshotGate_0.OnGateEnter)
   WavePlayer_0 = UE.Wave Player:Mono(Play=@ArrayRandomGet_0.On Next, Wave Asset=@ArrayRandomGet_0.Value)
   Multiply_0 = UE.Multiply:Audio by Float(
    AdditionalOperands=MapRange.MapRange:Float(In=^NormSpeed) {also {defaults {In Range A=0.3}}}.Out Value
    PrimaryOperand=UE.Multiply:Audio by Float(AdditionalOperands=^OneShotVolume, PrimaryOperand=@WavePlayer_0.Out Mono).Out
   )
  ]
  links [{dst="@OnOneShotPlayed", src=MSP_OneshotGate_0.OnGateEnter}, {dst="@Out L", src=Multiply_0.Out}]
 }
]
