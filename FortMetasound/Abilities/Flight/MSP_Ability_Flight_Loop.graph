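# MetaSound patch for a looping flight ability sound, playing a random wave from MovementSounds on OnPlay until OnStop.
# NormSpeed maps between min and max pitch mods, scaled by LoopPitch, while an interpolated FadeIn, SpeedVol and LoopVolume set gain.
# A RandomTime node relays the play trigger, but its half-duration random value is unconnected.
# The Mobile page uses a mono player into Out L only and omits LoopPitch.
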MetaSoundPatch /FortMetasound/Abilities/Flight/MSP_Ability_Flight_Loop
inputs [
 FadeIn {type=Float}
 LoopPitch {type=Float}
 LoopVolume {type=Float}
 MaxPitchMod {type=Float}
 MinPitchMod {type=Float}
 MovementSounds {type=WaveAsset:Array}
 NormSpeed {type=Float}
 OnPlay {type=Trigger}
 OnStop {type=Trigger}
 SpeedVol {type=Float}
]
outputs [
 Out L {type=Audio}
 Out R {type=Audio}
]
pages [
 Default {
  is_build_graph=true
  input_defaults {
   FadeIn=0.25
   LoopPitch=1
   LoopVolume=1
   MaxPitchMod=1.1
   MinPitchMod=0.9
   NormSpeed=0
   OnPlay=null
   OnStop=null
   SpeedVol=1
   MovementSounds []
  }
  nodes [
   n1 = UE.InterpTo:Audio(Target=^FadeIn) {also {defaults {Interp Time=0.5}}}
   ArrayRandomGet_0 = Array.Random Get:WaveAsset:Array(In Array=^MovementSounds, Next=^OnPlay)
   FrequencyMultipliertoSemitone_0 = UE.Frequency Multiplier to Semitone:Float(Frequency Multiplier=UE.Multiply:Float(
     AdditionalOperands=^LoopPitch
     PrimaryOperand=MapRange.MapRange:Float(Out Range B=^MaxPitchMod, Out Range A=^MinPitchMod, In=^NormSpeed).Out Value
    ).Out)
   GetWaveDuration_0 = "UE.Get Wave Duration: "(Wave=@ArrayRandomGet_0.Value)
   WavePlayer_0 = UE.Wave Player:Stereo(
    Stop=^OnStop
    Wave Asset=@ArrayRandomGet_0.Value
    Pitch Shift=@FrequencyMultipliertoSemitone_0.Semitones
    Play=UE.RandomTime(
     Next=@ArrayRandomGet_0.On Next
     Max=UE.Multiply:Time by Float(PrimaryOperand=@GetWaveDuration_0.Duration) {also {defaults {AdditionalOperands=0.5}}}.Out
    ).On Next
   ) {also {defaults {Loop="<MetaSoundBuilderResult.SUCCEEDED: 0>"}}}
   Multiply_0 = UE.Multiply:Audio by Float(
    AdditionalOperands=^LoopVolume
    PrimaryOperand=UE.Multiply:Audio by Float(
     AdditionalOperands=^SpeedVol
     PrimaryOperand=UE.Multiply:Audio by Float(AdditionalOperands=@n1.Value, PrimaryOperand=@WavePlayer_0.Out Left).Out
    ).Out
   )
   Multiply_3 = UE.Multiply:Audio by Float(
    AdditionalOperands=^LoopVolume
    PrimaryOperand=UE.Multiply:Audio by Float(
     AdditionalOperands=^SpeedVol
     PrimaryOperand=UE.Multiply:Audio by Float(AdditionalOperands=@n1.Value, PrimaryOperand=@WavePlayer_0.Out Right).Out
    ).Out
   )
  ]
  links [{dst="@Out L", src=Multiply_0.Out}, {dst="@Out R", src=Multiply_3.Out}]
 }
 Mobile {
  nodes [
   n1 = UE.InterpTo:Audio(Target=^FadeIn) {also {defaults {Interp Time=0.5}}}
   ArrayRandomGet_0 = Array.Random Get:WaveAsset:Array(In Array=^MovementSounds, Next=^OnPlay)
   FrequencyMultipliertoSemitone_0 = UE.Frequency Multiplier to Semitone:Float(Frequency Multiplier=MapRange.MapRange:Float(Out Range B=^MaxPitchMod, Out Range A=^MinPitchMod, In=^NormSpeed).Out Value)
   GetWaveDuration_0 = "UE.Get Wave Duration: "(Wave=@ArrayRandomGet_0.Value)
   Multiply_3 = UE.Multiply:Time by Float(PrimaryOperand=@GetWaveDuration_0.Duration) {also {defaults {AdditionalOperands=0.5}}}
   WavePlayer_0 = UE.Wave Player:Mono(
    Stop=^OnStop
    Wave Asset=@ArrayRandomGet_0.Value
    Pitch Shift=@FrequencyMultipliertoSemitone_0.Semitones
    Play=UE.RandomTime(Next=@ArrayRandomGet_0.On Next, Max=@Multiply_3.Out).On Next
   )
   Multiply_0 = UE.Multiply:Audio by Float(
    AdditionalOperands=^LoopVolume
    PrimaryOperand=UE.Multiply:Audio by Float(
     AdditionalOperands=^SpeedVol
     PrimaryOperand=UE.Multiply:Audio by Float(AdditionalOperands=@n1.Value, PrimaryOperand=@WavePlayer_0.Out Mono).Out
    ).Out
   )
  ]
  links [{dst="@Out L", src=Multiply_0.Out}]
 }
]
